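# Editor utility Blueprint that holds the options for a helper-asset placement tool in the Workshop editor tools.
# An enum variable selects which helper asset to place, defaulting to a player height reference.
# Its Remove All Placed Assets function only broadcasts the OnRemoveAllClicked dispatcher, leaving the removal to whatever binds to it.
# The event graph is empty, and no placement logic lives in this asset.
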EditorUtilityBlueprint /Game/Blueprints/Blutilities/ToolsWrkshp/ScrTools/HelperAssets/BP_STPS_HelperAssets
graphs [
 EventGraph {links [], nodes []}
 OnRemoveAllClicked {nodes [OnRemoveAllClicked() {is=title K2Node_FunctionEntry}]}
 Remove All Placed Assets {nodes [
   Remove All Placed Assets() {is=title K2Node_FunctionEntry}
   > Call On Remove All Clicked!(execute:Exec, self:BP STPS Helper Assets Object Reference) -> (then:Exec) {is=title K2Node_CallDelegate}
  ]}
]
variables [
 Asset to Place {type=E_HelperAssets_AssetChoices, default=Player Height Reference}
 OnRemoveAllClicked {default=()}
]
