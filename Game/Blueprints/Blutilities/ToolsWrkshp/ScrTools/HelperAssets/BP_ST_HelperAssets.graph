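# Editor scriptable tool that places scale-reference mannequins or color calibrators in the level on left click.
# Click traces the world, spawns the chosen asset as a tagged, optionally transient, editor-only actor, then offsets it by mesh bounds to sit on the ground.
# Remove All deletes tagged actors in a transaction after a suppressible warning.
# Setup restores saved property sets, adds the header widget, and turns off Game View; Esc exits.
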EditorUtilityBlueprint /Game/Blueprints/Blutilities/ToolsWrkshp/ScrTools/HelperAssets/BP_ST_HelperAssets
graphs [
 EventGraph {nodes [
   n1 = Get Header Widget()
   n2 = AddExpandableBox(self=@n1)
   > n3 = Set Readout Box(Readout Box:BP UW Expandable Box Object Reference=@n2) -> (Output_Get:BP UW Expandable Box Object Reference)
   > SetUpBox(
    self=@n3
    Name=Get Readout Box Label!(self:BP UW Shared Header Widget Object Reference=@n1) -> (Readout Box Label:Text)
    OpenByDefault=Get ReadoutOpenOnExit!(self:BP STPS Shared Widget State Object Reference=Get SharedState Settings() -> (SharedState Settings:BP STPS Shared Widget State Object Reference)) -> (ReadoutOpenOnExit:bool)
   )
   > n4 = Sequence()
   n5 = Set Distance Line(execute=@n6.then, Distance Line:BP UW Readout Line Object Reference=@n6) -> (Output_Get:BP UW Readout Line Object Reference)
   n6 = SetupLine!(
    execute=@n4.then_0
    self:BP UW Expandable Box Object Reference=Get Readout Box() -> (Readout Box:BP UW Expandable Box Object Reference)
    CurrentLine:BP UW Readout Line Object Reference=@n5
    DynamicLine:Vertical Box Object Reference
    Label:Text=Get Widget Distance Line Label() -> (Widget Distance Line Label:Text)
    Value:Text=Get Widget Distance Line Value() -> (Widget Distance Line Value:Text)
   ) -> (then, Readout Line:BP UW Readout Line Object Reference)
   n7 = Set Angle Line(execute=@n8.then, Angle Line:BP UW Readout Line Object Reference=@n8) -> (Output_Get:BP UW Readout Line Object Reference)
   n8 = SetupLine(
    execute=@n4.then_1
    self=Get Readout Box()
    CurrentLine=@n7
    Label=Get Widget Angle Rotated Line Label() -> (Widget Angle Rotated Line Label:Text)
    Value=Get Widget Angle Rotated Line Value() -> (Widget Angle Rotated Line Value:Text)
   )
   Event OnScriptSetup()
   > n9 = Sequence!(execute) -> (then_0, then_1, then_2:Exec, then_3:Exec, then_4:Exec)
   > K2Node_Composite_4 = ConfigPropertySets!(execute) -> (then:Exec) {is=title K2Node_Composite}
   K2Node_Composite_0 = Add Behaviours(execute=@n9.then_1) {is=title K2Node_Composite}
   K2Node_Composite_2 = HandleGameMode(execute=@n9.then_2) {is=title K2Node_Composite}
   K2Node_Composite_3 = BindToPropSetActions(execute=@n9.then_3) {is=title K2Node_Composite}
   K2Node_Composite_1 = Add Header Widget(execute=@n9.then_4) {is=title K2Node_Composite}
   n10 = Event OnScriptTick()
   > Set CurTime(CurTime=float + float(A=@n10.DeltaTime, B=Get CurTime()))
   > n11 = Sequence!(execute) -> (then_0, then_1, then_2)
   n12 = Get HaveAlreadyWarnedGameView() -> (HaveAlreadyWarnedGameView:bool)
   > Branch(Condition=AND Boolean(A=Get Is Game View On Tool Start() -> (Is Game View On Tool Start:bool), B=NOT Boolean(A=@n12)))
   > WarnGameViewOn()
   n13 = TimedClearUserWarning!(
    execute=@n11.then_1
    self
    ClearAfterTime:double=Get Clear After Time() -> (Clear After Time:double)
    CurTime:double=Get CurTime()
    LastClearTime:double=Get Last Clear Time() -> (Last Clear Time:double)
   ) -> (then, NewLastClearTime:double)
   > Set Last Clear Time(Last Clear Time:double=@n13) -> (Output_Get:double)
   n14 = Branch(execute=@n11.then_2, Condition=Get FirstInitRun())
   Set FirstInitRun(execute=@n14.else)
   > n15 = Sequence!(execute) -> (then_0, then_1)
   n16 = Get Header Widget()
   > n17 = AddExpandableBox(self=@n16)
   > n18 = Set Shortcut Box(Shortcut Box:BP UW Expandable Box Object Reference=@n17) -> (Output_Get:BP UW Expandable Box Object Reference)
   > n19 = SetUpBox(
    self=@n18
    Name=Get Shortcut Box Label!(self:BP UW Shared Header Widget Object Reference=@n16) -> (Shortcut Box Label:Text)
    OpenByDefault=Get ShortcutsOpenOnExit!(self:BP STPS Shared Widget State Object Reference=Get SharedState Settings()) -> (ShortcutsOpenOnExit:bool)
   )
   n20 = Set ShortcutLines(execute=@n21.then, ShortcutLines:Array of BP UW Readout Line Object References=@n21) -> (Output_Get:Array of BP UW Readout Line Object References)
   n21 = AddShortcuts!(
    execute=@n19.then
    self:BP UW Expandable Box Object Reference=Get Shortcut Box() -> (Shortcut Box:BP UW Expandable Box Object Reference)
    ShortcutMap:Map of Strings to Strings=Get Shortcut Map() -> (Shortcut Map:Map of Strings to Strings)
    ExistingShortcutLines:Array of BP UW Readout Line Object References=@n20
    ShortcutMapKeys:Array of Strings=Get Shortcut Map Keys() -> (Shortcut Map Keys:Array of Strings)
   ) -> (then, AllShortcutLines:Array of BP UW Readout Line Object References)
   n22 = AddExpandableBox(execute=@n15.then_1, self=Get Header Widget())
   > n23 = Set Actions Box(Actions Box=@n22)
   > SetUpBox(self=@n23, Name=Get Actions Box Label(self=Get Header Widget()))
   > n24 = Construct Object from Class(Class=&/BP_UW_ActionsButton_HelperAssets.BP_UW_ActionsButton_HelperAssets_C) -> (ReturnValue:BP UW Actions Button Helper Assets Object Reference)
   > n25 = Set ActionButtons(ActionButtons:BP UW Actions Button Helper Assets Object Reference=@n24) -> (Output_Get:BP UW Actions Button Helper Assets Object Reference)
   > AddChildToVerticalBox(self=Get DynamicEntries(self=Get Actions Box()), Content=@n25)
   > Bind Event to Event Remove All Clicked(
    self:BP UW Actions Button Helper Assets Object Reference=Get ActionButtons() -> (ActionButtons:BP UW Actions Button Helper Assets Object Reference)
    Delegate=Create Event()
   ) {is=title K2Node_AddDelegate}
   Event OnScriptShutdown()
   > K2Node_Composite_5 = SavePropSetStates() {is=title K2Node_Composite}
  ]}
 Add Behaviours {
  path=EventGraph.K2Node_Composite_0.Add Behaviours
  nodes [
   Inputs()
   > AddSingleClickBehavior!(
    execute
    self:Scriptable Modular Behavior Tool Object Reference
    IfHitByClick:Delegate=Create Event()
    OnHitByClick:Delegate=Create Event()
    CaptureCheck:Delegate
    CapturePriority:int=100
    MouseButton:EScriptableToolMouseButton Enum=LeftButton
    bHitTestOnRelease:bool=true
   ) -> (then)
   > AddSingleKeyInputBehavior!(
    execute
    self:Scriptable Modular Behavior Tool Object Reference
    OnKeyPressed:Delegate=Create Event()
    OnKeyReleased:Delegate
    OnForceEndCaptureFuncIn:Delegate
    Key:Key Structure=Escape
    CaptureCheck:Delegate
    CapturePriority:int=1000000
   ) -> (then)
   Outputs()
  ]
 }
 Add Header Widget {
  path=EventGraph.K2Node_Composite_1.Add Header Widget
  nodes [
   Inputs()
   > n1 = Sequence()
   > n2 = Construct Object from Class(Class=&BP_UW_SharedHeaderWidget_C) -> (ReturnValue:BP UW Shared Header Widget Object Reference)
   > n3 = Set Header Widget(Header Widget:BP UW Shared Header Widget Object Reference=@n2) -> (Output_Get:BP UW Shared Header Widget Object Reference)
   > SetToolPanelToolHeaderWidget(Widget=@n3)
   DisplayUserWarningMessage(execute=@n1.then_1, Message="LOCTEXT(\" \")")
   Outputs()
  ]
 }
 HandleGameMode {
  path=EventGraph.K2Node_Composite_2.HandleGameMode
  nodes [
   Inputs()
   > n1 = Construct Object from Class(Class=/Game/Blueprints/Blutilities/ToolsWrkshp/ScrTools/TravelTime/SupportFiles/BP_TravelTime_EditorOnlyUtils.BP_TravelTime_EditorOnlyUtils_C) -> (ReturnValue:BP Travel Time Editor Only Utils Object Reference)
   > n2 = Set EditorOnlyTools(EditorOnlyTools:BP Travel Time Editor Only Utils Object Reference=@n1) -> (Output_Get:BP Travel Time Editor Only Utils Object Reference)
   > n3 = GetSetGameView!(execute, self:BP Travel Time Editor Only Utils Object Reference=@n2, SetGameView:bool=false, GameViewVal:bool=false) -> (then, ViewportKey:Name=None, IsGameView:bool)
   > n4 = Set Is Game View On Tool Start(Is Game View On Tool Start:bool=@n3.IsGameView) -> (Output_Get:bool)
   > Branch(Condition=@n4)
   > Set HaveAlreadyWarnedGameView(HaveAlreadyWarnedGameView:bool=false) -> (Output_Get:bool)
   > GetSetGameView(self=Get EditorOnlyTools() -> (EditorOnlyTools:BP Travel Time Editor Only Utils Object Reference), SetGameView=true, GameViewVal=false) -> (ViewportKey=None)
   Outputs()
  ]
 }
 BindToPropSetActions {
  path=EventGraph.K2Node_Composite_3.BindToPropSetActions
  nodes [
   Inputs()
   > Bind Event to On Remove All Clicked(self:BP STPS Helper Assets Object Reference=Get Settings() -> (Settings:BP STPS Helper Assets Object Reference), Delegate=Create Event()) {is=title K2Node_AddDelegate}
   Outputs()
  ]
 }
 ConfigPropertySets {
  path=EventGraph.K2Node_Composite_4.ConfigPropertySets
  nodes [
   Inputs()
   > n1 = Sequence!(execute) -> (then_0, then_1, then_2:Exec, then_3:Exec)
   n2 = AddPropertySetOfType(
    execute=@n1.then_1
    PropertySetType=/Game/Blueprints/Blutilities/ToolsWrkshp/ScrTools/Shared/SharedPropertySets/BP_STPS_SharedWidgetState.BP_STPS_SharedWidgetState_C
    Identifier=Get SharedStateIdentifier() -> (SharedStateIdentifier:String)
   )
   > n3 = Cast To BP_STPS_SharedWidgetState!(execute, Object=@n2) -> (then, CastFailed, AsBP STPS Shared Widget State:BP STPS Shared Widget State Object Reference) {is=title K2Node_DynamicCast}
   > Set SharedState Settings(SharedState Settings:BP STPS Shared Widget State Object Reference=@n3) -> (Output_Get:BP STPS Shared Widget State Object Reference)
   > SetPropertySetVisibleByName!(execute, self:Scriptable Interactive Tool Object Reference, Identifier:String=Get SharedStateIdentifier(), bVisible:bool=false) -> (then)
   > RestorePropertySetSettings(PropertySet=Get SharedState Settings(), SaveKey=Get SaveSettings_Identifier() -> (SaveSettings_Identifier:String))
   Outputs!(then:Exec=@n1.then_3)
   n4 = AddPropertySetOfType(
    execute=@n1.then_0
    PropertySetType=&/BP_STPS_HelperAssets.BP_STPS_HelperAssets_C
    Identifier=Get Settings_Identifier() -> (Settings_Identifier:String)
   )
   > n5 = Cast To BP_STPS_HelperAssets!(execute, Object=@n4) -> (then, CastFailed, AsBP STPS Helper Assets:BP STPS Helper Assets Object Reference) {is=title K2Node_DynamicCast}
   > Set Settings(Settings:BP STPS Helper Assets Object Reference=@n5) -> (Output_Get:BP STPS Helper Assets Object Reference)
   > RestorePropertySetSettings(PropertySet=Get Settings(), SaveKey=Get SaveSettings_Identifier())
  ]
 }
 SavePropSetStates {
  path=EventGraph.K2Node_Composite_5.SavePropSetStates
  nodes [
   Inputs()
   > SavePropertySetSettings(PropertySet=Get Settings(), SaveKey=Get SaveSettings_Identifier())
   n1 = Get Shortcut Box()
   > Set ShortcutsOpenOnExit!(
    execute
    ShortcutsOpenOnExit:bool=IsChecked(self=Get ExpandButton!(self:BP UW Expandable Box Object Reference=@n1) -> (ExpandButton:Check Box Object Reference))
    self:BP STPS Shared Widget State Object Reference=Get SharedState Settings()
   ) -> (then, Output_Get:bool)
   > SavePropertySetSettings(PropertySet=Get SharedState Settings(), SaveKey=Get SaveSettings_Identifier())
   Outputs!()
  ]
 }
 IfHitByClick {nodes [
   n1 = IfHitByClick!() -> (
    then
    ClickPos_WorldRay:Ray Structure
    ClickPos_bHas2D:bool=~
    ClickPos_ScreenPosition:Vector 2D Structure
    MouseButton:EScriptableToolMouseButton Enum=LeftButton
   ) {is=title K2Node_FunctionEntry}
   n2 = GetRayStartEnd!(Ray:Ray Structure=@n1, StartDistance:double=0, EndDistance:double=0) -> (StartPoint:Vector, EndPoint:Vector)
   > n3 = Line Trace By Channel!(
    execute
    Start:Vector=@n2
    End:Vector=@n2.EndPoint
    TraceChannel:ETraceTypeQuery Enum
    bTraceComplex:bool=true
    ActorsToIgnore:Array of Actor Object References
    DrawDebugType:EDrawDebugTrace Enum
    bIgnoreSelf:bool
    TraceColor:Linear Color Structure
    TraceHitColor:Linear Color Structure
    DrawTime:float
   ) -> (
    then
    OutHit_bBlockingHit:bool
    OutHit_bInitialOverlap:bool
    OutHit_Time:float
    OutHit_Distance:float
    OutHit_Location:Vector
    OutHit_ImpactPoint:Vector
    OutHit_Normal:Vector
    OutHit_ImpactNormal:Vector
    OutHit_PhysMat:Physical Material Object Reference
    OutHit_HitActor:Actor Object Reference
    OutHit_HitComponent:Primitive Component Object Reference
    OutHit_HitBoneName:Name=None
    OutHit_BoneName:Name=None
    OutHit_HitItem:int=0
    OutHit_ElementIndex:int=0
    OutHit_FaceIndex:int=0
    OutHit_TraceStart:Vector
    OutHit_TraceEnd:Vector
    ReturnValue:bool
   )
   > Set HitLoc(HitLoc:Vector=@n3.OutHit_Location) -> (Output_Get:Vector)
   > Return Node!(
    execute
    ReturnValue_bHit:bool=@n3.ReturnValue
    ReturnValue_HitDepth:double=340282346638528859811704183484516925440
    ReturnValue_HitNormal:Vector="0,0,1"
    ReturnValue_bHasHitNormal:bool=False
    ReturnValue_HitIdentifier:int=0
    ReturnValue_HitObject:Object Reference
   )
  ]}
 On Esc Pressed {nodes [
   On Esc Pressed!() -> (then, Key:Key Structure=None, Modifiers:Scriptable Tool Modifier States Structure) {is=title K2Node_FunctionEntry}
   > RequestToolShutdown!(execute, self:Scriptable Interactive Tool Object Reference, bAccept:bool=true, bShowUserPopupMessage:bool=false, UserMessage:Text) -> (then)
  ]}
 OnClearAllUtilityActors {nodes [
   OnClearAllUtilityActors() {is=title K2Node_FunctionEntry}
   > n1 = GetAllActorsWithTag!(execute, Tag:Name=Get ScaleActorTagName() -> (ScaleActorTagName:Name=None)) -> (then, OutActors:Array of Actor Object References)
   n2 = Length(TargetArray=@n1)
   > n3 = Branch(Condition="integer > integer"(A=@n2))
   n4 = "To Text (String)"(InString=Get ClearTransientActorsWarnIniName() -> (ClearTransientActorsWarnIniName:String))
   n5 = Get ToolName(None=Self-Reference()) -> (ToolName:Text)
   n6 = "Not Equal (Integer)"(A=@n2, B=1)
   n7 = Get ClearTransientActorsWarnMessage_p() -> (ClearTransientActorsWarnMessage_p:Text)
   n8 = Get ClearTransientActorsWarnMessage_s() -> (ClearTransientActorsWarnMessage_s:Text)
   > n9 = Show Suppressable Warning Dialog!(
    execute
    Title:"Text (by ref)"=Get ClearTransientActorsWarnTitle() -> (ClearTransientActorsWarnTitle:Text)
    Message:"Text (by ref)"=Format Text!(Format=SelectText(A=@n7, B=@n8, bPickA=@n6), 0:int=@n2) -> (Result)
    InIniSettingName:String="To String (Text)"(InText=Format Text(Format=@n4, 0:Text=@n5))
    InIniSettingFileNameOverride:String
    bDefaultValue:bool=true
   ) -> (then, ReturnValue:bool)
   > Branch(Condition=@n9)
   > BeginTransaction(Context=Get TransactionContext() -> (TransactionContext:String), Description="LOCTEXT(\"Delete Actor\")")
   > n10 = For Each Loop(Array:Array of Actor Object References=@n1) -> (Array Element:Actor Object Reference)
   EndTransaction(execute=@n10.Completed)
   Destroy Actor(execute=@n10.LoopBody, self=@n10)
   DisplayUserWarningMessage(execute=@n3.else, Message=Get ClearTransientNoneFoundWarn() -> (ClearTransientNoneFoundWarn:Text))
   > Set Last Clear Time(Last Clear Time=Get CurTime())
   Get ClearTransientActorsWarnIniName()
  ]}
 OnHitByClick {nodes [
   OnHitByClick!() -> (
    then
    ClickPos:Input Device Ray Structure
    Modifiers:Scriptable Tool Modifier States Structure
    MouseButton:EScriptableToolMouseButton Enum=LeftButton
   ) {is=title K2Node_FunctionEntry}
   > n1 = Sequence!(execute) -> (then_0, then_1)
   > n2 = Switch on E_HelperAssets_AssetChoices!(
    execute
    Selection:E_HelperAssets_AssetChoices Enum=Get Asset to Place!(self:BP STPS Helper Assets Object Reference=Get Settings()) -> (Asset to Place:E_HelperAssets_AssetChoices Enum=NewEnumerator1)
   ) -> (NewEnumerator1:Exec, NewEnumerator0:Exec) {is=title K2Node_SwitchEnum}
   Set ActorFinalName(execute=@n2.NewEnumerator0, ActorFinalName:String=Get ActorNameColor() -> (ActorNameColor:String)) -> (Output_Get:String)
   > n3 = SpawnActorFromClass!(
    execute
    ActorClass:Actor Class Reference=&Actor
    Location:Vector=Get HitLoc() -> (HitLoc:Vector)
    Rotation:Rotator
    bTransient:bool=Get Transient() -> (Transient:bool)
   ) -> (then, ReturnValue:Actor Object Reference)
   > n4 = Set AddedScaleActor(AddedScaleActor:Actor Object Reference=@n3) -> (Output_Get:Actor Object Reference)
   > n5 = Gather Subobject Data For Instance(self=Get SubobjectDataSubsystem(), Context=@n4)
   > Set Common Actor Properties!(execute:Exec, actor:Actor Object Reference=Get AddedScaleActor() -> (AddedScaleActor:Actor Object Reference)) -> (then:Exec) {is=title K2Node_MacroInstance}
   > n6 = AddNewSubobject!(
    execute
    self:Subobject Data Subsystem Object Reference=Get SubobjectDataSubsystem()
    Params:"Add New Subobject Params Structure (by ref)"=Make Add New Subobject Params!(
     ParentHandle:Subobject Data Handle Structure="Get (a copy)"(Array:Array of Subobject Data Handle Structures=@n5, Dimension 1=1) -> (Output:Subobject Data Handle Structure)
     NewClass:Object Class Reference=&StaticMeshComponent
     BlueprintContext:Blueprint Object Reference
     bSkipMarkBlueprintModified:bool=False
     bConformTransformToParent:bool=True
    ) -> (AddNewSubobjectParams:Add New Subobject Params Structure) {is=title K2Node_MakeStruct}
   ) -> (then, FailReason:Text, ReturnValue:Subobject Data Handle Structure)
   > n7 = Cast To StaticMeshComponent(Object=GetAssociatedObject!(Data:"Subobject Data Structure (by ref)"=GetData(DataHandle=@n6.ReturnValue)) -> (ReturnValue:Object Reference))
   > SetEditorProperty(Object=@n7, PropertyName=bIsEditorOnly, PropertyValue=Get Editor Only() -> (Editor Only:bool))
   > SetStaticMesh(self=@n7, NewMesh=Get ColorCalibMesh() -> (ColorCalibMesh:Static Mesh Object Reference))
   > Add Unique(
    TargetArray:Array of Actor Object References=Get AllAddedScaleActors() -> (AllAddedScaleActors:Array of Actor Object References)
    NewItem:"Actor Object Reference (by ref)"=Get AddedScaleActor()
   )
   Set ActorFinalName(execute=@n2.NewEnumerator1, ActorFinalName=Get ActorNameScaleActor() -> (ActorNameScaleActor:String))
   > n8 = SpawnActorFromClass(ActorClass=/Script/Engine.SkeletalMeshActor, Location=Get HitLoc(), bTransient=Get Transient()) -> (ReturnValue:Skeletal Mesh Actor Object Reference)
   > Set AddedScaleActor(AddedScaleActor=@n8)
   > n9 = Set AddedSkelMeshActor(AddedSkelMeshActor:Skeletal Mesh Actor Object Reference=@n8) -> (Output_Get:Skeletal Mesh Actor Object Reference)
   n10 = Get SkeletalMeshComponent!(self:Skeletal Mesh Actor Object Reference=@n9) -> (SkeletalMeshComponent:Skeletal Mesh Component Object Reference)
   > SetMaterial(self=@n10, ElementIndex=0, Material=/Game/Creative/Devices/Mannequin/Materials/MI_CP_Device_Mannequin_02)
   > SetSkeletalMeshAsset!(
    execute
    self:Skeletal Mesh Component Object Reference=@n10
    NewMesh:Skeletal Mesh Object Reference=Get Mannequin_SKM() -> (Mannequin_SKM:Skeletal Mesh Object Reference)
   ) -> (then)
   > Set Common Actor Properties(actor=Get AddedSkelMeshActor() -> (AddedSkelMeshActor:Skeletal Mesh Actor Object Reference))
   > Add Unique(TargetArray=Get AllAddedScaleActors(), NewItem=Get AddedScaleActor())
   PlaceActorBaseOnGround!(execute=@n1.then_1, self, Actor:Actor Object Reference=Get AddedScaleActor()) -> (then)
  ]}
 PlaceActorBaseOnGround {nodes [
   n1 = PlaceActorBaseOnGround!() -> (then, Actor:Actor Object Reference) {is=title K2Node_FunctionEntry}
   n2 = Cast To StaticMeshActor!(execute, Object=@n1) -> (then, CastFailed, AsStatic Mesh Actor:Static Mesh Actor Object Reference) {is=title K2Node_DynamicCast}
   n3 = GetComponentByClass!(self:Actor Object Reference=@n1, ComponentClass:Actor Component Class Reference=&StaticMeshComponent) -> (ReturnValue:Static Mesh Component Object Reference)
   n4 = GetBounds!(self:Static Mesh Object Reference=Get StaticMesh(self=@n3)) -> (ReturnValue_Origin:Vector, ReturnValue_BoxExtent:Vector, ReturnValue_SphereRadius:float)
   n5 = Set Delta Z(execute=[@n2.then, @n9.Is Valid], Delta Z:double=float - float(A=BreakVector(InVec=@n4.ReturnValue_BoxExtent).Z, B=BreakVector(InVec=@n4).Z)) -> (Output_Get:double)
   Add Actor World Offset!(
    execute=[@n8.then, @n5.then]
    self:Actor Object Reference=@n1
    DeltaLocation_X:double=0
    DeltaLocation_Y:double=0
    DeltaLocation_Z:double=Get Delta Z() -> (Delta Z:double)
    bSweep:bool=false
    bTeleport:bool=false
   ) -> (then, SweepHitResult:Hit Result Structure)
   n6 = GetComponentByClass(self=@n1, ComponentClass=/Script/Engine.SkeletalMeshComponent) -> (ReturnValue:Skeletal Mesh Component Object Reference)
   Is Valid(exec=[@n2.CastFailed, @n9.Is Not Valid], InputObject=@n6)
   n7 = GetBounds(self:Skeletal Mesh Object Reference=GetSkeletalMeshAsset!(self:Skeletal Mesh Component Object Reference=@n6) -> (ReturnValue:Skeletal Mesh Object Reference))
   > n8 = Set Delta Z(Delta Z=float - float(A=BreakVector(InVec=@n7.ReturnValue_BoxExtent).Z, B=BreakVector(InVec=@n7).Z))
   Get StaticMeshComponent(self=@n2)
   n9 = Is Valid(exec=@n1.then, InputObject=@n3)
  ]}
 SetCommonActorProperties {nodes [
   n1 = Inputs!() -> (execute, actor:Actor Object Reference)
   > SetActorLabel(self=@n1, NewActorLabel=Get ActorFinalName() -> (ActorFinalName:String), bMarkDirty=false)
   > SetEditorProperty(Object=@n1, PropertyName=bIsEditorOnlyActor, PropertyValue=Get Editor Only())
   > Set Tags!(
    execute
    Tags:Array of Names=Make Array("[0]":Name=Get ScaleActorTagName() -> (ScaleActorTagName=None)) -> (Array:Array of Names)
    self:Actor Object Reference=@n1
   ) -> (then, Output_Get:Array of Names)
   > SetSelectedLevelActors(self=Get EditorActorSubsystem(), ActorsToSelect=Make Array("[0]":Actor Object Reference=@n1) -> (Array:Array of Actor Object References))
   > Outputs!(then:Exec)
  ]}
 TimedClearUserWarning {nodes [
   n1 = TimedClearUserWarning!() -> (then, ClearAfterTime:double=3, CurTime:double=~, LastClearTime:double=~) {is=title K2Node_FunctionEntry}
   > Set LastClearTimeUpdated(LastClearTimeUpdated:double=@n1.LastClearTime) -> (Output_Get:double)
   > Branch(Condition="float > float"(A=@n1.CurTime, B=float + float(A=@n1, B=@n1.LastClearTime)))
   > Branch()
   > Set LastClearTimeUpdated(LastClearTimeUpdated=@n1.CurTime)
   > ClearUserMessages!(execute, self:Scriptable Interactive Tool Object Reference, bHelpMessage:bool=true, bWarningMessage:bool=true) -> (then)
   > Return Node!(execute, NewLastClearTime:double=Get LastClearTimeUpdated() -> (LastClearTimeUpdated:double))
  ]}
 WarnGameViewOn {nodes [
   WarnGameViewOn() {is=title K2Node_FunctionEntry}
   > Branch(Condition=Get Is Game View On Tool Start())
   > DisplayUserWarningMessage(Message=Get GameViewOnWarnMessage() -> (GameViewOnWarnMessage:Text))
   > Set Last Clear Time(Last Clear Time=Get CurTime())
  ]}
]
variables [
 ActionButtons {type=BP_UW_ActionsButton_HelperAssets_C, default=null}
 Actions Box {type=BP_UW_ExpandableBox_C}
 ActorFinalName {type=string, default=""}
 AddedSkelMeshActor {type=SkeletalMeshActor, default=null}
 AllAddedScaleActors {type="array<Actor>"}
 Clear After Time {type=number, default=7}
 ClearTransientActorsWarnIniName {type=string, default="ToolsWorkshop_{0}_ClearAllWarning"}
 ClearTransientActorsWarnMessage_p {type=string, default="This will remove {0} Helper Assets in the level. Delete anyway?"}
 ClearTransientActorsWarnMessage_s {type=string, default=This will remove 1 Helper Asset in the level. Delete anyway?}
 ClearTransientActorsWarnTitle {type=string, default=Clear Helper Assets}
 ClearTransientNoneFoundWarn {type=string, default=No objects found to remove.}
 ColorCalibMesh {type=StaticMesh, default=/Engine/EditorMeshes/ColorCalibrator/SM_ColorCalibrator}
 CurTime {type=number}
 Editor Only {type=boolean, default=true}
 EditorOnlyTools {type=BP_TravelTime_EditorOnlyUtils_C, default=null}
 FirstInitRun {type=boolean}
 GameViewOnWarnMessage {type=string, default=Game View in the viewport has been turned off to view Helper Assets. Press G in the viewport to turn it back on.}
 HaveAlreadyWarnedGameView {type=boolean}
 Header Widget {type=BP_UW_SharedHeaderWidget_C}
 HitLoc {type=Vector}
 Is Game View On Tool Start {type=boolean}
 Last Clear Time {type=number}
 Mannequin_SKM {type=SkeletalMesh, default=/Game/Creative/Devices/Mannequin/Meshes/CP_Device_Mannequin}
 Mannequin_SM {type=StaticMesh, default=/Game/Creative/Devices/Mannequin/Meshes/CP_Device_Mannequin_01}
 PropSetSaveKey {type=string, default=""}
 SaveSettings_Identifier {type=string, default=SaveSettings}
 ScaleActorTagName {type=string, default=ScaleActor}
 Settings {type=BP_STPS_HelperAssets_C, default=null}
 Settings_Identifier {type=string, default=Settings}
 SharedState Settings {type=BP_STPS_SharedWidgetState_C, default=null}
 SharedStateIdentifier {type=string, default=SharedStateSettings}
 Shortcut Box {type=BP_UW_ExpandableBox_C}
 Shortcut Map {type="map<string>", default {Place Helper Asset=Left Click}}
 Shortcut Map Keys {type="array<string>", default [Place Helper Asset]}
 ShortcutLines {type="array<BP_UW_ReadoutLine_C>", default []}
 TransactionContext {type=string, default=""}
 Transient {type=boolean, default=true}
]
